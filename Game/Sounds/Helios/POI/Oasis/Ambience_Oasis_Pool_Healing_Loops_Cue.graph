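# Looping ambience cue for the healing pool at the Helios Oasis POI, using ambient attenuation and sound class.
# A Random node picks one of two equally weighted mixes, each layering the Slurp Vat heal loop through a Modulator with a different Sakura Springs pool energy loop.
# Looping Envelopers add a gentle volume swell over about 31 seconds.
# The root QualityLevel node routes to two identical Envelopers sharing that Random, so all quality levels sound alike.
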SoundCue /Game/Sounds/Helios/POI/Oasis/Ambience_Oasis_Pool_Healing_Loops_Cue
root=QualityLevel_0
nodes [
 {
  id=Enveloper_2
  children [Random_0]
  properties {
   bLoop=true
   bLoopIndefinitely=true
   loopEnd=31.49
   pitchCurve {editorCurveData {keys [{time=0}]}}
   volumeCurve {editorCurveData {keys [
      {arriveTangent=-0.01, interpMode=RCIM_Cubic, time=0, value=1}
      <at 5 1.2>
      {interpMode=RCIM_Cubic, leaveTangent=-0.0133, time=10, value=1.1}
      {interpMode=RCIM_Cubic, leaveTangent=0.01, time=31.49, value=1}
     ]}}
  }
 }
 {
  id=Enveloper_4
  children [Random_0]
  properties {
   bLoop=true
   bLoopIndefinitely=true
   loopEnd=31.49
   pitchCurve {editorCurveData {keys [{time=0}]}}
   volumeCurve {editorCurveData {keys [
      {arriveTangent=-0.01, interpMode=RCIM_Cubic, time=0, value=1}
      <at 5 1.2>
      {interpMode=RCIM_Cubic, leaveTangent=-0.0133, time=10, value=1.1}
      {interpMode=RCIM_Cubic, leaveTangent=0.01, time=31.49, value=1}
     ]}}
  }
 }
 {id=Mixer_0, children [WavePlayer_7,Modulator_1], properties {inputVolume [1.5,5.5]}}
 {id=Mixer_2, children [Modulator_0,WavePlayer_9], properties {inputVolume [5,1.25]}}
 {id=Modulator_0, children [WavePlayer_1], properties {pitchMax=1.1, pitchMin=1}}
 {id=Modulator_1, children [WavePlayer_2], properties {pitchMax=1, pitchMin=1}}
 {id=QualityLevel_0, children [Enveloper_2,Enveloper_2,Enveloper_4,Enveloper_4]}
 {id=Random_0, children [Mixer_2,Mixer_0], properties {weights [1,1]}}
 {id=WavePlayer_1, properties {bLooping=true, soundWaveAssetPtr=/Game/Athena/Sounds/Items/Slurp/Item_Slurp_Vat_Heal_Loop_02}}
 {id=WavePlayer_2, properties {bLooping=true, soundWaveAssetPtr=&Item_Slurp_Vat_Heal_Loop_02}}
 {
  id=WavePlayer_7
  properties {bLooping=true, soundWaveAssetPtr=/Game/Sounds/Asteria/POI/SakuraSprings/Spots/Spot_Sakura_Springs_Slurp_Pool_Energy_3_LQ}
 }
 {
  id=WavePlayer_9
  properties {bLooping=true, soundWaveAssetPtr=/Game/Sounds/Asteria/POI/SakuraSprings/Spots/Spot_Sakura_Springs_Slurp_Pool_Energy_4_LQ}
 }
]
properties {
 attenuationSettings=&Fort_Ambient_Athena_Att
 maxDistance=2097152
 soundClassObject=&Fort_Ambient_Athena
 volumeMultiplier=0.1
 preEffectBusSends [{}]
}
