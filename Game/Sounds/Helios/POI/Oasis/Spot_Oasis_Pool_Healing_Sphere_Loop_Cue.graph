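# Looping outdoor ambient spot sound cue for the Oasis POI pool's healing sphere, reusing Sakura Springs slurp pool energy loops.
# It randomly picks one of two looping waves with slight random pitch, feeding every quality level identically.
# A slow volume oscillator and two nested looping envelopers add minute-long pitch and half-minute volume cycles.
# Overridden attenuation adds listener focus, and concurrency caps instances by stopping the farthest.
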SoundCue /Game/Sounds/Helios/POI/Oasis/Spot_Oasis_Pool_Healing_Sphere_Loop_Cue
root=Enveloper_2
nodes [
 {
  id=Enveloper_1
  children [Oscillator_0]
  properties {
   bLoop=true
   bLoopIndefinitely=true
   loopEnd=60
   pitchCurve {editorCurveData {keys [{interpMode=RCIM_Cubic, time=0}, <at 30 1.025 RCIM_Cubic>, {interpMode=RCIM_Cubic, time=60}]}}
   volumeCurve {editorCurveData {keys [<at 0 1>]}}
  }
 }
 {
  id=Enveloper_2
  children [Enveloper_1]
  properties {
   bLoop=true
   bLoopIndefinitely=true
   loopEnd=30
   pitchCurve {editorCurveData {keys [{time=0}]}}
   volumeCurve {editorCurveData {keys [
      {arriveTangent=-0.01, interpMode=RCIM_Cubic, time=0, value=1}
      <at 15 0.85 RCIM_Cubic>
      {interpMode=RCIM_Cubic, leaveTangent=0.01, time=30, value=1}
     ]}}
  }
 }
 {id=Mixer_1, children [WavePlayer_2], properties {inputVolume [0.8]}}
 {id=Modulator_1, children [Random_0], properties {pitchMax=1.1, pitchMin=1}}
 {
  id=Oscillator_0
  children [QualityLevel_0]
  properties {amplitudeMax=0.4, amplitudeMin=0.2, bModulateVolume=true, centerMax=1, centerMin=1, frequencyMax=0.05, frequencyMin=0.03}
 }
 {id=QualityLevel_0, children [Modulator_1,Modulator_1,Modulator_1,Modulator_1]}
 {id=Random_0, children [WavePlayer_4,Mixer_1], properties {weights [1,1]}}
 {
  id=WavePlayer_2
  properties {bLooping=true, soundWaveAssetPtr=/Game/Sounds/Asteria/POI/SakuraSprings/Spots/Spot_Sakura_Springs_Slurp_Pool_Energy_2_LQ}
 }
 {
  id=WavePlayer_4
  properties {bLooping=true, soundWaveAssetPtr=/Game/Sounds/Asteria/POI/SakuraSprings/Spots/Spot_Sakura_Springs_Slurp_Pool_Energy_1_LQ}
 }
]
properties {
 attenuationSettings=/Game/Sounds/Attenuation/Fort_Looping_Spot_Att
 bOverrideAttenuation=true
 bOverrideConcurrency=true
 maxDistance=5457.2563
 soundClassObject=&Fort_Ambient_Outdoors_Athena
 volumeMultiplier=0.1
 attenuationOverrides {
  bEnableListenerFocus=true
  nonFocusAzimuth=90
  nonFocusVolumeAttenuation=0.4
  nonSpatializedRadiusEnd=1000
  nonSpatializedRadiusStart=1500
  stereoSpread=4000
  attenuationShapeExtents {x=1857.2564}
 }
 concurrencyOverrides {maxCount=2, resolutionRule=StopFarthestThenPreventNew}
}
